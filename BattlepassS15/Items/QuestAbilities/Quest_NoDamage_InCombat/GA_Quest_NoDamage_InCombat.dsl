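; Passive ability supporting a Battle Pass no-damage-in-combat quest, tied to the owning player's lifetime and quest objective.
; On activation it caches the avatar pawn, then uses a server check when binding owner death and quest objective-state events.
; Owner death, or an update to the objective named in Q_ObjectiveName, can trigger Cleanup, which unbinds both events and ends the ability.
; No damage tracking happens in this graph.
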(event Ability|EventFailedToActivatePassiveAbility
  (params (ActorInfo :type Gameplay%20Ability%20Actor%20Info%20Structure)))

(event Ability|EventActivateAbility
  (bind _owning_player (Variables|Default|GetOwning%20Player))
  ;[0 0 0] Activate
  (bind _asfort_player_pawn_athena (Utilities|Casting|CastToFortPlayerPawnAthena (Ability|GetAvatarActorfromActorInfo))
    (:then
      (Variables|Default|SetOwning%20Player _asfort_player_pawn_athena)
      (if (Networking|IsServer)
        (EventDispatchers|BindEventtoOnDied _owning_player (EventDispatchers|CreateEvent :!event Custom|OnOwnerDied))
        (EventDispatchers|BindEventtoOnQuestObjectiveStateChanged (Fortnite|GetFortPlayerControllerfromActor _owning_player) (EventDispatchers|CreateEvent :!event Custom|Objective%20Updated))))
    (:CastFailed
      (Ability|EndAbility))))

(event Custom|Cleanup
  (bind _owning_player (Variables|Default|GetOwning%20Player))
  ;[0 0 0] Cleanup Handles and Binds
  (EventDispatchers|UnbindEventfromOnDied _owning_player (EventDispatchers|CreateEvent :!event Custom|OnOwnerDied))
  (EventDispatchers|UnbindEventfromOnQuestObjectiveStateChanged (Fortnite|GetFortPlayerControllerfromActor _owning_player) (EventDispatchers|CreateEvent :!event Custom|Objective%20Updated))
  (Ability|EndAbility))

(event Custom|OnOwnerDied
  (params (DamagedActor :type Actor%20Object%20Reference) (Damage :type Float%20%28single-precision%29) (InstigatedBy :type Controller%20Object%20Reference) (DamageCauser :type Actor%20Object%20Reference) (HitLocation :type Vector) (FHitComponent :type Primitive%20Component%20Object%20Reference) (BoneName :type Name) (Momentum :type Vector))
  ;[0 0 0] Fail Quest if Owner Dies
  (CallFunction|Cleanup))

(event Custom|Objective%20Updated
  (params (Controller :type Fort%20Player%20Controller%20Object%20Reference) (QuestDef :type Fort%20Quest%20Item%20Definition%20Object%20Reference) (BackendName :type Name) (CompletionCount :type Integer) (ObjectiveCompleted :type Boolean) (QuestCompleted :type Boolean))
  ;[0 0 0] On Objective Complete Cleanup
  (if (Utilities|Name|Equal(Name) BackendName (Variables|Default|GetQ_ObjectiveName))
    (if ObjectiveCompleted
      (Development|PrintString "Objective Complete")
      (CallFunction|Cleanup))))
